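# Physics asset for the NightHawk backpack cosmetic, giving its skeletal mesh a simple two-body rig for physical motion.
# A short capsule on the dyn_hinge bone and a long, thin capsule on the pack bone are joined by one constraint.
# The constraint sets a nonzero limit only on the second swing axis, suggesting a narrow single-axis hinge for the pack.
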PhysicsAsset /BRCosmetics/Accessories/FORT_Backpacks/Backpack_M_MED_NightHawk/Meshes/M_MED_NightHawk_Pack_Physics
bodies {
 dyn_hinge [{shapeType=Capsule, length=2, radius=2, shapeName=dyn_hinge_capsule, center=<1 0 0>, rotation=<rot 90 0 0>}]
 pack [{shapeType=Capsule, length=13.8125, radius=0.7807, shapeName=pack_capsule, center=<-2 0 0>, rotation=<rot 0 0 -90>}]
}
constraints [{bone1Name=dyn_hinge, bone2Name=pack, swing1LimitDegrees=0, swing2LimitDegrees=20, twistLimitDegrees=0}]
