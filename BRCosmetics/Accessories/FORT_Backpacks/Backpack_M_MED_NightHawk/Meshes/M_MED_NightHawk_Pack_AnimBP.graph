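# Animation Blueprint for the MED NightHawk backpack cosmetic, adding physics sway and body and movement adjustments.
# The pack bone is offset, adjusted for Large body types found on customization load, then fed through two RigidBody nodes: gameplay, or stiffer-clamp-free front end.
# Both share a physics asset and per-frame gravity override.
# Afterwards, a sprinting alpha, chosen from skydive dive, tactical sprint, sprint or regular state, tilts dyn_hinge and nudges the pack.
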AnimBlueprint /BRCosmetics/Accessories/FORT_Backpacks/Backpack_M_MED_NightHawk/Meshes/M_MED_NightHawk_Pack_AnimBP
target_skeleton=&/M_MED_NightHawk_Pack_Skeleton
graphs [
 AnimGraph {nodes [
   n1 = "Transform (Modify) Bone - Bone: pack"(ComponentPose=Local To Component(LocalPose=Input Pose()), Translation="1.2,3,0", Rotation="0,0,0") {
    rotationSpace=BCS_WorldSpace
    translationMode=BMM_Additive
    translationSpace=BCS_BoneSpace
    alphaBoolBlend {blendOption=Linear}
    alphaScaleBiasClamp {bInterpResult=true}
    boneToModify {boneName=pack}
    translation=<1.2 3>
   }
   n2 = "Transform (Modify) Bone - Bone: pack"(ComponentPose=@n1, Alpha="To Float (Boolean)"(InBool=Get IsLarge() -> (IsLarge:bool)), Translation="0,-5,-5", Rotation="0,0,0") {
    translationMode=BMM_Replace
    alphaBoolBlend {blendOption=Linear}
    alphaScaleBiasClamp {bInterpResult=true}
    boneToModify {boneName=pack}
    translation {y=-5, z=-5}
   }
   n3 = Get bIsInFrontEnd()
   n4 = Get GravityOverride()
   n5 = RigidBody!(ComponentPose=@n2, bAlphaBoolEnabled:bool=NOT Boolean(A=@n3), OverrideWorldGravity:Vector=@n4, ExternalForce) -> (Pose) {
    alphaInputType=Bool
    bClampLinearTranslationLimitToRefPose=true
    bOverrideWorldGravity=true
    overridePhysicsAsset=&/M_MED_NightHawk_Pack_Physics
    simulationSpace=BaseBoneSpace
    alphaBoolBlend {blendOption=Linear}
    baseBoneRef {boneName=root}
    componentAppliedLinearAccClamp=<3000 3000 3000>
    componentLinearAccScale=<0.4 0.3 0.3>
    componentLinearVelScale=<0.4 0.3 0.3>
    overrideWorldGravity {z=-980}
    simSpaceSettings {
     maxAngularAcceleration=1000
     maxAngularVelocity=3
     maxLinearAcceleration=700
     maxLinearVelocity=500
     velocityScaleZ=0
     worldAlpha=0.65
     externalLinearDragV=<0.2 0.2>
    }
   }
   n6 = RigidBody(ComponentPose=@n5, bAlphaBoolEnabled=@n3, OverrideWorldGravity=@n4) {
    alphaInputType=Bool
    bClampLinearTranslationLimitToRefPose=true
    bOverrideWorldGravity=true
    overridePhysicsAsset=&M_MED_NightHawk_Pack_Physics
    simulationSpace=BaseBoneSpace
    alphaBoolBlend {blendOption=Linear}
    baseBoneRef {boneName=root}
    componentAppliedLinearAccClamp=<5000 5000 5000>
    componentLinearAccScale=<0.55 0.45 0.45>
    componentLinearVelScale=<0.55 0.45 0.45>
    overrideWorldGravity {z=-980}
    simSpaceSettings {
     maxAngularAcceleration=1000
     maxAngularVelocity=3
     maxLinearAcceleration=700
     maxLinearVelocity=500
     velocityScaleZ=0
     worldAlpha=1
     externalLinearDragV=<0.2 0.2>
    }
   }
   n7 = Get sprintingAlpha() -> (sprintingAlpha:float)
   n8 = "Transform (Modify) Bone - Bone: dyn_hinge"(ComponentPose=@n6, Alpha=@n7, Translation="0,0,0", Rotation="-25,0,0", Scale="1,1,1") {
    rotationMode=BMM_Additive
    rotationSpace=BCS_BoneSpace
    translationSpace=BCS_WorldSpace
    alphaBoolBlend {blendOption=Linear}
    alphaScaleBiasClamp {bInterpResult=true}
    boneToModify {boneName=dyn_hinge}
    rotation {pitch=-25}
   }
   n9 = "Transform (Modify) Bone - Bone: pack"(ComponentPose=@n8, Alpha=@n7, Translation="0,0,1.5", Rotation="0,0,0") {
    rotationSpace=BCS_WorldSpace
    translationMode=BMM_Additive
    translationSpace=BCS_BoneSpace
    alphaBoolBlend {blendOption=Linear}
    alphaScaleBiasClamp {bInterpResult=true}
    boneToModify {boneName=pack}
    translation {z=1.5}
   }
   Output Pose(Result=Component To Local(ComponentPose=@n9)) {layerGroup=None}
  ]}
 EventGraph {nodes [
   Event BlueprintUpdateAnimation()
   > n1 = Sequence()
   > n2 = Gravity Override State Function DEPRECATED(CharacterPartAnimInstance=Self-Reference(), GameTimeInSeconds=GetGameTimeInSeconds(), GravityOverrideParams=Get Gravity Override Params())
   > Set GravityOverride(GravityOverride=@n2)
   n3 = Branch(execute=@n1.then_1, Condition=OR Boolean!(A=false, B=Get bIsSkydiveDiveMode(), C:bool=Get bIsTacticalSprinting()) -> (ReturnValue))
   > Set sprintingAlpha(sprintingAlpha:float=Get TacticalSprinting Alpha() -> (TacticalSprinting Alpha:float)) -> (Output_Get:float)
   n4 = Branch(execute=@n3.else, Condition=OR Boolean(A=Get bIsSprinting(), B=false))
   > Set sprintingAlpha(sprintingAlpha=Get Sprinting Alpha_0() -> (Sprinting Alpha_0:float))
   Set sprintingAlpha(execute=@n4.else, sprintingAlpha=Get RegularAlpha() -> (RegularAlpha:float))
   Event OnCharacterCustomizationLoaded()
   > n5 = Set Owning Pawn(Owning Pawn:Pawn Object Reference=TryGetPawnOwner()) -> (Output_Get:Pawn Object Reference)
   > Is Valid(InputObject=@n5)
   > n6 = Sequence()
   n7 = FindBodyType!(execute=@n6.then_1, Object:Object Reference=Get Owning Pawn() -> (Owning Pawn:Pawn Object Reference)) -> (True:Exec, False:Exec, OutBodyType:EFortCustomBodyType Enum=NONE)
   > Set IsLarge(IsLarge:bool="Equal (Enum)"(A:EFortCustomBodyType Enum=@n7, B:EFortCustomBodyType Enum=Large)) -> (Output_Get:bool)
   n8 = Cast To FortPlayerPawn(execute=@n6.then_0, Object=Get Owning Pawn())
   > n9 = Set Owning Player Pawn(Owning Player Pawn:Fort Player Pawn Object Reference=@n8) -> (Output_Get:Fort Player Pawn Object Reference)
   > Set BodyMesh(BodyMesh=GetSkeletalMeshForPartType(self=@n9))
   TryGetPawnOwner()
  ]}
 NewFunction {nodes [NewFunction() {is=title K2Node_FunctionEntry}]}
]
variables [
 As Player Pawn Athena {type=PlayerPawn_Athena_C}
 BodyMesh {type=SkeletalMeshComponent}
 fallingAlpha {type=number, default=1}
 Gravity Override Params {type=GravityOverrideParamsStruct, default {front_End {gravityOverride=<0 0 -350>}}}
 GravityOverride {type=Vector}
 isFemale {type=boolean}
 IsLarge {type=boolean}
 Owning Pawn {type=Pawn, default=null}
 Owning Player Pawn {type=FortPlayerPawn}
 RegularAlpha {type=number}
 Rotation {type=Rotator}
 Sprinting Alpha_0 {type=number, default=0.5}
 sprintingAlpha {type=number, default=1}
 TacticalSprinting Alpha {type=number, default=1}
]
